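; Gameplay ability that teleports an unaware AI bot back near its spawn or last teleport spot when it strays outside its leash.
; A looping timer and alert-level changes trigger the check, gated by minimum pawn lifetime and an in-progress flag.
; A spawned querier actor anchors an EQS search for the destination, moving there on success.
; The pawn teleports hidden between start/end gameplay cues, falling back to SetActorLocation and logging if both fail.
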(event Ability|EventActivateAbility
  (bind _outfortplayerpawn (CallFunction|GetAvatarasValidFortPlayerPawn)
    (:ValidFortPlayerPawn
      (bind _output_get (Variables|InternalData|SetAIPawn _outfortplayerpawn))
      (Utilities|IsValid (Pawn|GetController _output_get)
        (:Is%20Valid
          (bind _asfort_athena_aibot_controller (Utilities|Casting|CastToFortAthenaAIBotController (Pawn|GetController _output_get))
            (:then
              (bind _output_get_2 (Variables|InternalData|SetAIController _asfort_athena_aibot_controller))
              (Variables|InternalData|SetOriginalLocation (Transformation|GetActorLocation (Variables|InternalData|GetAIPawn)))
              (Variables|Default|SetisTeleporting false)
              (AlertLevel|BindEventtoOnAlertLevelChangedEvent _output_get_2 (EventDispatchers|CreateEvent :!event Custom|AIAlertLevelChanged))
              (bind _returnvalue (Utilities|Time|SetTimerbyEvent (EventDispatchers|CreateEvent :!event Custom|ValidateLeash) (ScalableFloat|GetValueatLevel (Variables|Default|GetValidateLeashTimer)) true))
              (Variables|InternalData|SetHNDL_ValidateLeashTimer _returnvalue)
              (bind _returnvalue_2 (Game|SpawnActorfromClass "/NPCLibrary/Abilities/LeashTeleport/Blueprints/BP_LeashTeleportQuerier.BP_LeashTeleportQuerier_C" (Transformation|GetActorTransform (Variables|InternalData|GetAIPawn)) "AlwaysSpawn" "MultiplyWithRoot" (Variables|InternalData|GetAIPawn) (Variables|InternalData|GetAIPawn)))
              (Variables|InternalData|SetOriginActorUsedForEQS _returnvalue_2))
            (:CastFailed
              _)))
        (:Is%20Not%20Valid
          _))
      (Ability|EndAbility))
    (:AvatarCastFailed
      _)))

(event Ability|EventOnEndAbility
  (params (bWasCancelled :type Boolean))
  (bind _originactorusedforeqs (Variables|InternalData|GetOriginActorUsedForEQS))
  (Utilities|Time|ClearandInvalidateTimerbyHandle (Variables|InternalData|GetHNDL_ValidateLeashTimer))
  (Utilities|IsValid _originactorusedforeqs
    (:Is%20Valid
      (Actor|DestroyActor _originactorusedforeqs))))

(event Ability|EventFailedToActivatePassiveAbility
  (params (ActorInfo :type Gameplay%20Ability%20Actor%20Info%20Structure)))

(event Custom|AIAlertLevelChanged
  (params (BotController :type Fort%20Athena%20AIBot%20Controller%20Object%20Reference) (OldAlertLevel :type EAlertLevel%20Enum :default "Unaware") (NewAlertLevel :type EAlertLevel%20Enum :default "Unaware"))
  (bind _aicontroller (Variables|InternalData|GetAIController))
  (bind _aipawn (Variables|InternalData|GetAIPawn))
  (bind _output_get (Variables|InternalData|SetLatestAlertLevel NewAlertLevel))
  (if (Utilities|Enum|Equal(Enum) _output_get "Unaware")
    (Utilities|IsValid _aicontroller
      (:Is%20Valid
        (Utilities|IsValid _aipawn
          (:Is%20Valid
            (if (Class|FortAthenaAIBotController|IsLocationInsideLeash _aicontroller (Transformation|GetActorLocation _aipawn) false 0.0)
              _
              (elif (> (Actor|GetGameTimeSinceCreation _aipawn) (ScalableFloat|GetValueatLevel (Variables|Default|GetLifeTimeThreshold)))
                (if (not (Variables|Default|GetisTeleporting))
                  (Variables|Default|SetisTeleporting true)
                  (CallFunction|FindTeleportLocation))))))))))

(event Custom|ValidateLeash
  (bind _aicontroller (Variables|InternalData|GetAIController))
  (bind _aipawn (Variables|InternalData|GetAIPawn))
  (if (Utilities|Enum|Equal(Enum) (Variables|InternalData|GetLatestAlertLevel) "Unaware")
    (Utilities|IsValid _aicontroller
      (:Is%20Valid
        (Utilities|IsValid _aipawn
          (:Is%20Valid
            (if (Class|FortAthenaAIBotController|IsLocationInsideLeash _aicontroller (Transformation|GetActorLocation _aipawn) false 0.0)
              _
              (elif (> (Actor|GetGameTimeSinceCreation _aipawn) (ScalableFloat|GetValueatLevel (Variables|Default|GetLifeTimeThreshold)))
                (if (not (Variables|Default|GetisTeleporting))
                  (Variables|Default|SetisTeleporting true)
                  (CallFunction|FindTeleportLocation))))))))))

(event Custom|FindTeleportLocation
  (bind _originactorusedforeqs (Variables|InternalData|GetOriginActorUsedForEQS))
  ;[1 1 1] Try to find a location on NavMesh where there's no AI near where we either spawned or last teleported to.
  (Utilities|IsValid _originactorusedforeqs
    (:Is%20Valid
      (bind _returnvalue (AI|EQS|RunEQSQuery (Variables|Default|GetEQS_TeleportDestination) _originactorusedforeqs))
      (AI|EQS|BindEventtoOnQueryFinished _returnvalue (delegate Custom|TeleportEQSFinished)))
    (:Is%20Not%20Valid
      (CallFunction|TeleportAI))))

(event Custom|TeleportEQSFinished
  (params (QueryInstance :type EQS%20Query%20Instance%20Object%20Reference) (QueryStatus :type EEnvQueryStatus%20Enum :default "Processing"))
  ;[1 1 1] Try to find a location on NavMesh where there's no AI near where we either spawned or last teleported to.
  (bind (_resultlocations _returnvalue) (AI|EQS|GetQueryResultsasLocations QueryInstance))
  (if _returnvalue
    (if (not (Math|Vector|VectorIsNearlyZero (Utilities|Array|Get(acopy) _resultlocations)))
      (bind _output_get (Variables|InternalData|SetOriginalLocation (Utilities|Array|Get(acopy) _resultlocations)))
      (Transformation|SetActorLocation (Variables|InternalData|GetOriginActorUsedForEQS) _output_get false true))
    (CallFunction|TeleportAI)))

(event Custom|TeleportAI
  (bind _originallocation (Variables|InternalData|GetOriginalLocation))
  (bind _aipawn (Variables|InternalData|GetAIPawn))
  ;[1 1 1] Try to teleport the AI to a location back inside the fight arena.
  (Ability|ExecuteGameplayCueWithParamsOnOwner (Variables|Default|GetGC_StartTeleport) (Ability|GameplayCue|MakeGameplayCueParameters :NormalizedMagnitude 0.0 :RawMagnitude 0.0 :Location (Transformation|GetActorLocation (Variables|InternalData|GetAIPawn)) :Normal "0,0,0" :GameplayEffectLevel 1 :AbilityLevel 1 :bReplicateLocationWhenUsingMinimalRepProxy false))
  (Rendering|SetActorHiddenInGame _aipawn true)
  (bind _returnvalue (Transform|Actor|TeleportPlayerPawn _aipawn _originallocation (Transformation|GetActorRotation _aipawn)))
  (if (not _returnvalue)
    (bind (_sweephitresult _returnvalue_2) (Transformation|SetActorLocation (Variables|InternalData|GetAIPawn) _originallocation false true))
    (if (not _returnvalue_2)
      (Fortnite|FortShippingLog (Variables|InternalData|GetAIPawn) (Utilities|String|ToString(Text) (Utilities|Text|FormatText "NSLOCTEXT(\"[BE1376D24C8F969F246E2DBD179DF361]\", \"A99335FB4924B9DF71FB499B56B8FF83\", \"AI {a} was outside his leash and GA_LeashTeleport tried to telport him back to the Arena, but the Teleport Pawn AND Set Actor Location both FAILED to teleport the AI.\")" (Utilities|GetDisplayName (Variables|InternalData|GetAIPawn)))) true)))
  (Utilities|FlowControl|Delay 0.2)
  (Ability|ExecuteGameplayCueWithParamsOnOwner (Variables|Default|GetGC_EndTeleport) (Ability|GameplayCue|MakeGameplayCueParameters :NormalizedMagnitude 0.0 :RawMagnitude 1e+01 :Location (Variables|InternalData|GetOriginalLocation) :Normal "0,0,0" :GameplayEffectLevel (Variables|Default|GetGameplayCueLevelToIdentifyAI) :AbilityLevel 1 :bReplicateLocationWhenUsingMinimalRepProxy false))
  (Rendering|SetActorHiddenInGame (Variables|InternalData|GetAIPawn) false)
  (Variables|Default|SetisTeleporting false))
